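# Looping quad MetaSound source that plays random volcano lava one-shots (flares, hisses, metal, rumbles) as ambience.
# On play, a random no-repeat array pick feeds a quad wave player, whose finish trigger loops back through a variable to pick the next sound.
# An azimuth panner patch positions the output by listener azimuth, with gain mapped from attenuation distance.
# A random-time trigger delay exists, but its output is unconnected.
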MetaSoundSource /Game/Athena/Sounds/Ambient/Volcano/MSS_AMB_Lava_OneShots
inputs [
 Sounds {type=WaveAsset:Array}
 UE.Attenuation.Distance {type=Float}
 UE.Listener.Orientation.Azimuth {type=Float}
 UE.Listener.Orientation.Elevation {type=Float}
 UE.Source.OnPlay {type=Trigger}
]
interfaces [UE.Attenuation:1.0, UE.Listener.Orientation:1.0, UE.OutputFormat.Quad:1.0, UE.Source:1.1]
outputs [
 "UE.OutputFormat.Quad.Audio:0" {type=Audio}
 "UE.OutputFormat.Quad.Audio:1" {type=Audio}
 "UE.OutputFormat.Quad.Audio:4" {type=Audio}
 "UE.OutputFormat.Quad.Audio:5" {type=Audio}
]
pages [Default {
  input_defaults {
   UE.Attenuation.Distance=0
   UE.Listener.Orientation.Azimuth=0
   UE.Listener.Orientation.Elevation=0
   UE.Source.OnPlay=null
   Sounds [
    &/AMB_Lava_Caldera_Spots_Flare_01
    &/AMB_Lava_Caldera_Spots_Flare_02
    &/AMB_Lava_Caldera_Spots_Flare_03
    &/AMB_Lava_Caldera_Spots_Flare_04
    &/AMB_Lava_Caldera_Spots_Flare_05
    &/AMB_Lava_Caldera_Spots_Flare_06
    &/AMB_Lava_Caldera_Spots_Flare_07
    &/AMB_Lava_Caldera_Spots_Flare_08
    &/AMB_Lava_Caldera_Spots_Flare_09
    &/AMB_Lava_Caldera_Spots_Flare_10
    &/AMB_Lava_Caldera_Spots_Flare_11
    &/AMB_Lava_Caldera_Spots_Flare_12
    &/AMB_Lava_Caldera_Spots_Flare_13
    &/AMB_Lava_Caldera_Spots_Hiss_01
    &/AMB_Lava_Caldera_Spots_Hiss_02
    &/AMB_Lava_Caldera_Spots_Hiss_03
    &/AMB_Lava_Caldera_Spots_Hiss_04
    &/AMB_Lava_Caldera_Spots_Hiss_05
    &/AMB_Lava_Caldera_Spots_Hiss_06
    &/AMB_Lava_Caldera_Spots_Hiss_07
    &/AMB_Lava_Caldera_Spots_Metal_01
    &/AMB_Lava_Caldera_Spots_Metal_02
    &/AMB_Lava_Caldera_Spots_Metal_03
    &/AMB_Lava_Caldera_Spots_Metal_04
    &/AMB_Lava_Caldera_Spots_Metal_05
    &/AMB_Lava_Caldera_Spots_Metal_06
    &/AMB_Lava_Caldera_Spots_Metal_07
    &/AMB_Lava_Caldera_Spots_Metal_08
    &/AMB_Lava_Caldera_Spots_Rumble_A_01
    &/AMB_Lava_Caldera_Spots_Rumble_A_02
    &/AMB_Lava_Caldera_Spots_Rumble_A_03
    &/AMB_Lava_Caldera_Spots_Rumble_A_04
    &/AMB_Lava_Caldera_Spots_Rumble_A_05
    &/AMB_Lava_Caldera_Spots_Rumble_A_06
    &/AMB_Lava_Caldera_Spots_Rumble_A_07
    &/AMB_Lava_Caldera_Spots_Rumble_A_08
    &/AMB_Lava_Caldera_Spots_Rumble_A_09
    &/AMB_Lava_Caldera_Spots_Rumble_A_10
   ]
  }
  nodes [
   n1 = VariableDeferredAccessor.Trigger(Variable=InitVariable.Trigger().Variable)
   TriggerAnyTriggerAny2_0 = "TriggerAny.Trigger Any (2)"(In 0=^UE.Source.OnPlay, In 1=@n1.Value)
   ArrayRandomGet_0 = Array.Random Get:WaveAsset:Array(In Array=^Sounds, No Repeats=Array.Num:WaveAsset:Array(Array=^Sounds).Num, Next=@TriggerAnyTriggerAny2_0.Out)
   WavePlayer_0 = UE.Wave Player:Quad(Play=@ArrayRandomGet_0.On Next, Wave Asset=@ArrayRandomGet_0.Value) {also {version [1,1]}}
   MSP_AzimuthPannerQuad_0 = &MSP_AzimuthPannerQuad(
    Azimuth=^UE.Listener.Orientation.Azimuth
    Gain=MapRange.MapRange:Float(In=^UE.Attenuation.Distance) {also {defaults {In Range A=10000, In Range B=0}}}.Out Value
    In Front Left=@WavePlayer_0.Out Front Left
    In Front Right=@WavePlayer_0.Out Front Right
    In Side Left=@WavePlayer_0.Out Side Left
    In Side Right=@WavePlayer_0.Out Side Right
   )
   n2 = UE.RandomTime(Next=@TriggerAnyTriggerAny2_0.Out) {also {defaults {Max=12, Min=6}}}
   TriggerDelay_0 = UE.Trigger Delay(In=@n2.On Next, Reset=@n2.On Reset, Delay Time=@n2.Value)
   VariableMutator.Trigger(Variable=@n1.Variable, Value=@WavePlayer_0.On Finished)
  ]
  links [
   {dst="@UE.OutputFormat.Quad.Audio:0", src=MSP_AzimuthPannerQuad_0.Out Front Left}
   {dst="@UE.OutputFormat.Quad.Audio:1", src=MSP_AzimuthPannerQuad_0.Out Front Right}
   {dst="@UE.OutputFormat.Quad.Audio:4", src=MSP_AzimuthPannerQuad_0.Out Side Left}
   {dst="@UE.OutputFormat.Quad.Audio:5", src=MSP_AzimuthPannerQuad_0.Out Side Right}
  ]
 }]
properties {
 assetImportData=&self:AssetImportData
 bLooping=true
 duration=10000
 numChannels=4
 outputFormat=Quad
 sampleRate=48000
 soundClassObject=&Fort_Ambient_OneShots
 volume=0.501
}
