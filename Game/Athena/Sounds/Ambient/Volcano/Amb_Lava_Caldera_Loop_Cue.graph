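# Looping ambient cue for the volcano's lava caldera, limited to one playing instance in the Athena ambient sound class.
# A mixer layers close gurgle, distant gurgle and a lava one-shots source, the last passed through a modulator for slight variation.
# Each layer is non-spatialized with its own custom distance curve, so close layers fade out with distance while the distant loop rises then fades, forming a distance-based crossfade with occlusion settings.
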SoundCue /Game/Athena/Sounds/Ambient/Volcano/Amb_Lava_Caldera_Loop_Cue
root=Mixer_0
nodes [
 {
  id=Attenuation_0
  children [WavePlayer_3]
  properties {
   attenuationSettings=&/AMB_Lava_Caldera_Close_Att
   attenuationOverrides {
    bEnableReverbSend=false
    bSpatialize=false
    distanceAlgorithm=Custom
    falloffDistance=12000
    occlusionInterpolationTime=0.5
    occlusionLowPassFilterFrequency=2000
    occlusionTraceChannel=ECC_GameTraceChannel11
    occlusionVolumeAttenuation=0.5
    customAttenuationCurve {editorCurveData {keys [<at 0 1>,<at 0.1289 0.8058>,<at 0.2461 0.8205>,<at 0.4 0>]}}
   }
  }
 }
 {
  id=Attenuation_1
  children [WavePlayer_4]
  properties {
   attenuationSettings=&/AMB_Lava_Caldera_Distant_Att
   attenuationOverrides {
    bEnableReverbSend=false
    bSpatialize=false
    distanceAlgorithm=Custom
    falloffDistance=6000
    occlusionInterpolationTime=0.5
    occlusionLowPassFilterFrequency=2000
    occlusionTraceChannel=ECC_GameTraceChannel11
    occlusionVolumeAttenuation=0.5
    attenuationShapeExtents {x=0}
    customAttenuationCurve {editorCurveData {keys [<at 0 0>, <at 0.2 1>, {interpMode=RCIM_Cubic, tangentMode=RCTM_User, time=1, value=0}]}}
   }
  }
 }
 {
  id=Attenuation_2
  children [WavePlayer_0]
  properties {
   attenuationSettings=&AMB_Lava_Caldera_Close_Att
   attenuationOverrides {
    bEnableReverbSend=false
    bSpatialize=false
    distanceAlgorithm=Custom
    falloffDistance=12000
    occlusionInterpolationTime=0.5
    occlusionLowPassFilterFrequency=2000
    occlusionTraceChannel=ECC_GameTraceChannel11
    occlusionVolumeAttenuation=0.5
    customAttenuationCurve {editorCurveData {keys [<at 0 1>,<at 0.1289 0.8058>,<at 0.2461 0.8205>,<at 0.4 0>]}}
   }
  }
 }
 {id=Mixer_0, children [Attenuation_0, Attenuation_1, Modulator_0], properties {inputVolume [2.35,1,2]}}
 {id=Modulator_0, children [Attenuation_2], properties {volumeMax=1}}
 {id=WavePlayer_0, properties {bLooping=true, soundWaveAssetPtr=&/MSS_AMB_Lava_OneShots}}
 {id=WavePlayer_3, properties {bLooping=true, soundWaveAssetPtr=&/AMB_Lava_Caldera_Gurgle_Close_2D_Loop}}
 {id=WavePlayer_4, properties {bLooping=true, soundWaveAssetPtr=&/AMB_Lava_Caldera_Gurgle_Distant_2D_Loop}}
]
properties {
 maxDistance=10000
 soundClassObject=&Fort_Ambient_Athena
 volumeMultiplier=1.002
 concurrencyOverrides {maxCount=1, resolutionRule=PreventNew}
}
