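# Sound cue for the start of the Floor Is Lava volcano event, playing an eruption followed by distant earth rumbles.
# A root delay holds back a mixer that layers the eruption with four rumble waves, each on its own fixed delay to form a staggered sequence, the last pitched down.
# Overrides give it a very long falloff with manual reverb send and allow only one instance, blocking new triggers while it plays.
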SoundCue /Game/Athena/Sounds/Ambient/Volcano/Amb_Volcano_Lava_Start_Cue
root=Delay_4
nodes [
 {id=Delay_0, children [WavePlayer_1], properties {delayMax=6, delayMin=6}}
 {id=Delay_1, children [WavePlayer_3], properties {delayMax=8, delayMin=8}}
 {id=Delay_2, children [WavePlayer_2], properties {delayMax=10, delayMin=10}}
 {id=Delay_3, children [Modulator_0], properties {delayMax=12, delayMin=12}}
 {id=Delay_4, children [Mixer_0], properties {delayMax=6.5, delayMin=6.5}}
 {id=Mixer_0, children [WavePlayer_0,Delay_0,Delay_1,Delay_2,Delay_3], properties {inputVolume [1,1,1,1,1]}}
 {id=Modulator_0, children [WavePlayer_4], properties {pitchMax=0.8, pitchMin=0.8}}
 {id=WavePlayer_0, properties {soundWaveAssetPtr=&/Amb_Volcano_FloorIsLava_Erupt}}
 {id=WavePlayer_1, properties {soundWaveAssetPtr=&/Amb_Volcano_Earth_Move_Rumble_Dist_03}}
 {id=WavePlayer_2, properties {soundWaveAssetPtr=&/Amb_Volcano_Earth_Move_Rumble_Dist_01}}
 {id=WavePlayer_3, properties {soundWaveAssetPtr=&/Amb_Volcano_Earth_Move_Rumble_Dist_05}}
 {id=WavePlayer_4, properties {soundWaveAssetPtr=&/Amb_Volcano_Earth_Move_Rumble_Dist_02}}
]
properties {
 bOverrideAttenuation=true
 bOverrideConcurrency=true
 maxDistance=100000
 soundClassObject=&Fort_Events
 virtualizationMode=Disabled
 volumeMultiplier=0.6
 attenuationOverrides {falloffDistance=100000, manualReverbSendLevel=15, reverbSendMethod=Manual, stereoSpread=2000, attenuationShapeExtents=<0 1500>}
 concurrencyOverrides {maxCount=1, resolutionRule=PreventNew}
}
